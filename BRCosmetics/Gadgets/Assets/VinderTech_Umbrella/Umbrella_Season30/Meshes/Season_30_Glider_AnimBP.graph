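# Animation Blueprint for the Season 30 VinderTech umbrella glider skeleton that layers a jitter motion onto the glider.
# The AnimGraph takes an input pose and applies the Season30_Jitter_GLIDER sequence on top of it as an additive, then sends the result to the output pose.
# The EventGraph holds only unconnected default update and pawn-owner nodes, so no variables or logic drive the blend.
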AnimBlueprint /BRCosmetics/Gadgets/Assets/VinderTech_Umbrella/Umbrella_Season30/Meshes/Season_30_Glider_AnimBP
target_skeleton=&/Season_30_Glider_Skeleton
graphs [
 AnimGraph {nodes [
   AnimGraphNode_SubInput_0 = Input Pose()
   Output Pose(Result=Apply Additive(
     Base=@AnimGraphNode_SubInput_0
     Additive="Sequence Player (additive) 'Season30_Jitter_GLIDER'"() {sequence=/BRCosmetics/Gadgets/Assets/VinderTech_Umbrella/Umbrella_Season30/Anim/Season30_Jitter_GLIDER}
    ) {alphaBoolBlend {blendOption=Linear}}) {layerGroup=None}
  ]}
 EventGraph {nodes [
   TryGetPawnOwner()
   Event BlueprintUpdateAnimation()
  ]}
]
variables []
